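# Physics asset for the Pickaxe_22_SqueakToy two-handed melee pickaxe cosmetic mesh.
# It approximates the weapon with two capsule bodies, a slim one on the handle bone and a larger, wider one on the b_AR_Root bone.
# A single constraint joins the two bodies, with limits on both swing axes and twist.
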PhysicsAsset /BRCosmetics/Weapons/FORT_Melee/Pickaxe_22_SqueakToy/Meshes/Pickaxe_22_SqueakToy_PhysicsAsset
bodies {
 Melee_TwoHanded_Handle [{shapeType=Capsule, length=26.5308, radius=4.6543, center=<1.1195 -1.9267 -18.8789>, rotation=<rot -0.8034 -90 0.2353>}]
 b_AR_Root [{shapeType=Capsule, length=38.4585, radius=17.7645, center=<7.962 -1.0241 36.1039>, rotation=<rot 0 -90.0002 -88.9597>}]
}
constraints [{bone1Name=Melee_TwoHanded_Handle, bone2Name=b_AR_Root, swing1LimitDegrees=45, swing2LimitDegrees=45, twistLimitDegrees=45}]
